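# Editor utility widget for inspecting and debugging animation on Juno pawns in a PIE session.
# Opening the combo box lists pawns of configured classes; selection enables debug panels and shows player options for Juno players.
# Tick draws checked facing, capsule, speed and translation overlays, partly mirrored to the Visual Logger.
# It also plays a test montage and pushes layering-disable toggles into the minifig AnimBP, though those AnimInstance casts appear as bad nodes.
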EditorUtilityWidgetBlueprint /AnimInEdTools/TDiA_Tools/PawnAnimInfo/EUW_JunoPawnAnimInfo
graphs [
 EventGraph {nodes [
   "On Opening (SelectPawnComboBox)"() {is=title K2Node_ComponentBoundEvent}
   > ClearOptions(self:"Combo Box (String) Object Reference"=Get SelectPawnComboBox() -> (SelectPawnComboBox:Editor Utility Combo Box String Object Reference))
   > Clear!(
    execute
    TargetMap:Map of Strings to Fort Pawn Object References=Get PawnsFound() -> (PawnsFound:Map of Strings to Fort Pawn Object References)
   ) -> (then) {is=title K2Node_CallFunction}
   > n1 = GetPIEWorlds!(execute, bIncludeDedicatedServer:bool=false) -> (then, ReturnValue:Array of World Object References)
   > n2 = For Each Loop(Array:Array of Actor Class References=Get Look For Pawns of Type() -> (Look For Pawns of Type:Array of Actor Class References)) -> (Array Element:Actor Class Reference)
   n3 = "Get (a copy)"(Array:Array of World Object References=@n1) -> (Output:World Object Reference)
   > n4 = GetAllActorsOfClass!(execute, WorldContextObject:Object Reference=@n3, ActorClass:Actor Class Reference=@n2) -> (then, OutActors:Array of Actor Object References)
   > n5 = For Each Loop(Array:Array of Actor Object References=@n4) -> (Array Element:Actor Object Reference)
   > n6 = Cast To FortPawn!(execute, Object=@n5) -> (then, CastFailed, AsFort Pawn:Fort Pawn Object Reference) {is=title K2Node_DynamicCast}
   n7 = GetDisplayName(Object=@n5)
   n8 = Get PawnsFound()
   > n9 = Branch(Condition=Find(TargetMap:Map of Strings to Fort Pawn Object References=@n8, Key=@n7) -> (Value:Fort Pawn Object Reference).ReturnValue)
   Add!(
    execute=@n9.else
    TargetMap:Map of Strings to Fort Pawn Object References=@n8
    Key:"String (by ref)"=@n7
    Value:"Fort Pawn Object Reference (by ref)"=@n6
   ) -> (then) {is=title K2Node_CallFunction}
   > AddOption!(execute, self:"Combo Box (String) Object Reference"=Get SelectPawnComboBox(), Option:String=@n7) -> (then)
   n10 = "On Selection Changed (SelectPawnComboBox)"!() -> (OutputDelegate, then, SelectedItem:String, SelectionType:ESelectInfo Enum) {is=title K2Node_ComponentBoundEvent}
   Set CurrentPawn(execute=[@n10.then, @n15.then], CurrentPawn:Fort Pawn Object Reference=Find(TargetMap=Get PawnsFound(), Key=@n10.SelectedItem)) -> (Output_Get:Fort Pawn Object Reference)
   n11 = NOT Boolean(A="Equal Exactly (String)"(A=@n10.SelectedItem))
   > SetIsEnabled(self=Get PawnBlockBorder() -> (PawnBlockBorder:Border Object Reference), bInIsEnabled=@n11)
   > SetIsEnabled(self=Get SkelBlockBorder() -> (SkelBlockBorder:Border Object Reference), bInIsEnabled=@n11)
   > SetIsEnabled(self=Get DebugMontageBlockBorder() -> (DebugMontageBlockBorder:Border Object Reference), bInIsEnabled=@n11)
   n12 = Get CurrentPawn() -> (CurrentPawn:Fort Pawn Object Reference)
   > n13 = Bad cast node!(execute, Object=GetAnimInstance(self=Get Mesh(self=@n12))) -> (then, CastFailed) {is=title K2Node_DynamicCast}
   > Set IsJunoPlayer(IsJunoPlayer:bool=true) -> (Output_Get:bool)
   n14 = Get JunoPlayerBlockBorder() -> (JunoPlayerBlockBorder:Border Object Reference)
   > SetVisibility(self=@n14, InVisibility=Visible)
   Set IsJunoPlayer(execute=@n13.CastFailed, IsJunoPlayer=false)
   > SetVisibility(self=@n14, InVisibility=Collapsed)
   n15 = Reset Pawn Selection_Event() {is=title K2Node_CustomEvent}
   Event PreConstruct()
   > SetObject(self=Get MontagePropView() -> (MontagePropView:Single Property View Object Reference), NewObject=Self-Reference())
   Event Tick()
   > n16 = Is Valid(InputObject=@n3)
   Branch(execute=@n16.Is Not Valid, Condition=Get World Is Valid() -> (World Is Valid:bool))
   > Clear(TargetMap=Get PawnsFound())
   > ClearSelection(self:"Combo Box (String) Object Reference"=Get SelectPawnComboBox())
   > Call Reset Pawn Selection!(execute:Exec, self:EUW Juno Pawn Anim Info Object Reference) -> (then:Exec) {is=title K2Node_CallDelegate}
   > Set World Is Valid(World Is Valid:bool=false) -> (Output_Get:bool)
   Set World Is Valid(execute=@n16.Is Valid, World Is Valid=true)
   > n17 = Sequence!(execute) -> (then_0, then_1, then_2:Exec, then_3:Exec, then_4:Exec, then_5:Exec, then_6:Exec, then_7:Exec)
   n18 = Get CurrentPawn()
   n19 = Get RootComponent!(self:Actor Object Reference=@n18) -> (RootComponent:Scene Component Object Reference)
   > Set PawnRootComponent(PawnRootComponent:Scene Component Object Reference=@n19) -> (Output_Get:Scene Component Object Reference)
   > Set PawnWorldTr(PawnWorldTr:Vector=Get World Location!(self:Scene Component Object Reference=@n19) -> (ReturnValue:Vector) {is=title K2Node_CallFunction}) -> (Output_Get:Vector)
   > Set PawnForwardVector(PawnForwardVector:Vector=GetForwardVector!(self:Scene Component Object Reference=@n19) -> (ReturnValue:Vector)) -> (Output_Get:Vector)
   n20 = Get CapsuleComponent!(self:Character Object Reference=@n18) -> (CapsuleComponent:Capsule Collision Object Reference)
   > Set PawnCapsuleComponent(PawnCapsuleComponent:Capsule Collision Object Reference=@n20) -> (Output_Get:Capsule Collision Object Reference)
   > Set Pawn Capsule Half Height(Pawn Capsule Half Height:double=Get CapsuleHalfHeight!(self:Capsule Collision Object Reference=@n20) -> (CapsuleHalfHeight:float)) -> (Output_Get:double)
   > Set Pawn Capsule Radius(Pawn Capsule Radius:double=Get CapsuleRadius!(self:Capsule Collision Object Reference=@n20) -> (CapsuleRadius:float)) -> (Output_Get:double)
   > Set DebugTextOffsetZ(DebugTextOffsetZ:double=0) -> (Output_Get:double)
   Branch(execute=@n17.then_1, Condition=IsChecked(self=Get DrawPawnFacing() -> (DrawPawnFacing:Editor Utility Check Box Object Reference)))
   n21 = Get PawnWorldTr() -> (PawnWorldTr:Vector)
   n22 = vector + vector(
    A=@n21
    B=vector * vector(A=Get PawnForwardVector() -> (PawnForwardVector:Vector), B=Get Arrow Length() -> (Arrow Length:double)) -> (ReturnValue=~)
   )
   n23 = Get PawnColor() -> (PawnColor:Linear Color Structure)
   > VisLog Arrow!(
    execute
    WorldContextObject:Object Reference=Get CurrentPawn()
    SegmentStart:Vector=@n21
    SegmentEnd:Vector=@n22
    Text:String
    ObjectColor:Linear Color Structure=@n23
    CategoryName:Name=PawnAnimInfo
    bAddToMessageLog:bool=false
    ArrowHeadSize:float=8
   ) -> (then)
   > DrawDebugArrow!(
    execute
    WorldContextObject:Object Reference=@n3
    LineStart:Vector=@n21
    LineEnd:Vector=@n22
    ArrowSize:float=20
    LineColor:Linear Color Structure=@n23
    Duration:float=0
    Thickness:float=0.3
    DepthPriority:EDrawDebugSceneDepthPriorityGroup Enum=World
   ) -> (then)
   > DrawDebugString!(
    execute
    WorldContextObject:Object Reference=@n3
    TextLocation:Vector=vector + vector(A=@n22, B="0,0,14")
    Text:String=Pawn Facing
    TestBaseActor:Actor Object Reference
    TextColor:Linear Color Structure=@n23
    Duration:float=0
   ) -> (then)
   Branch(execute=@n17.then_2, Condition=IsChecked(self=Get DrawSkelFacing() -> (DrawSkelFacing:Editor Utility Check Box Object Reference)))
   n24 = GetSocketTransform!(self=Get Mesh(self=Get CurrentPawn()), InSocketName=root, TransformSpace=RTS_World) -> (ReturnValue_Location:Vector, ReturnValue_Rotation:Rotator, ReturnValue_Scale:Vector)
   n25 = GetRightVector!(InRot:Rotator=@n24.ReturnValue_Rotation) -> (ReturnValue:Vector)
   n26 = Get Arrow Length()
   n27 = float - float(A=Get Pawn Capsule Half Height() -> (Pawn Capsule Half Height:double), B=6)
   n28 = vector + vector!(
    A=vector + vector!(A=@n24, B:Vector=vector * vector(A=@n25, B=@n26) -> (ReturnValue=~)) -> (ReturnValue)
    B_X:double=0
    B_Y:double=0
    B_Z:double=@n27
   ) -> (ReturnValue)
   n29 = vector + vector!(A=@n24, B_X:double=0, B_Y:double=0, B_Z:double=@n27) -> (ReturnValue)
   n30 = Get SkelColor() -> (SkelColor:Linear Color Structure)
   > VisLog Arrow(
    WorldContextObject=Get CurrentPawn()
    SegmentStart=@n29
    SegmentEnd=@n28
    ObjectColor=@n30
    CategoryName=PawnAnimInfo
    bAddToMessageLog=false
    ArrowHeadSize=8
   )
   > DrawDebugArrow(WorldContextObject=@n3, LineStart=@n29, LineEnd=@n28, ArrowSize=20, LineColor=@n30, Duration=0, Thickness=0.3, DepthPriority=World)
   > DrawDebugString(
    WorldContextObject=@n3
    TextLocation=vector + vector!(A=@n28, B:Vector="0,0,-14") -> (ReturnValue)
    Text=Skel Facing
    TextColor=@n30
    Duration=0
   )
   Branch(execute=@n17.then_3, Condition=IsChecked(self=Get DrawPawnCapsule() -> (DrawPawnCapsule:Editor Utility Check Box Object Reference)))
   n31 = Get PawnCapsuleComponent() -> (PawnCapsuleComponent:Capsule Collision Object Reference)
   n32 = Get World Rotation!(self:Scene Component Object Reference=@n31) -> (ReturnValue:Rotator) {is=title K2Node_CallFunction}
   n33 = Get World Location(self=@n31)
   n34 = BreakVector(InVec=@n33)
   n35 = Get CapsuleHalfHeight(self=@n31)
   n36 = Get CapsuleRadius(self=@n31)
   n37 = Get PawnColor()
   > VisLog Capsule Shape!(
    execute
    WorldContextObject:Object Reference=Get CurrentPawn()
    Base:Vector=MakeVector(X=@n34, Y=@n34.Y, Z=float - float(A=@n34.Z, B=@n35))
    HalfHeight:float=@n35
    Radius:float=@n36
    Rotation:Quat Structure="ToQuaternion (Rotator)"!(InRot:Rotator=@n32) -> (ReturnValue:Quat Structure)
    Text:String
    ObjectColor:Linear Color Structure=@n37
    LogCategory:Name=PawnAnimInfo
    bAddToMessageLog:bool=false
    bWireframe:bool=true
   ) -> (then)
   > DrawDebugCapsule!(
    execute
    WorldContextObject:Object Reference=@n3
    Center:Vector=@n33
    HalfHeight:float=@n35
    Radius:float=@n36
    Rotation:Rotator=@n32
    LineColor:Linear Color Structure=@n37
    Duration:float=0
    Thickness:float=0.3
    DepthPriority:EDrawDebugSceneDepthPriorityGroup Enum=World
   ) -> (then)
   Branch(execute=@n17.then_4, Condition=IsChecked(self=Get DrawPawnSpeed2D() -> (DrawPawnSpeed2D:Editor Utility Check Box Object Reference)))
   n38 = GetAnimInstance(self=Get Mesh(self=Get CurrentPawn()))
   > n39 = Bad cast node!(execute, Object=@n38) -> (then, CastFailed, AsJuno Creature Anim Instance:Object Reference)
   n40 = IsValid(Object=@n39)
   n41 = Bad cast node!(execute=@n39.CastFailed, Object=@n38) -> (then, CastFailed, AsJuno Player Anim Instance:Object Reference)
   n42 = Get PawnSpeed2D!(self:Object Reference=@n41) -> (PawnSpeed2D)
   n43 = Get PawnSpeed2D(self=@n39)
   DrawDebugString(
    execute=[@n39.then, @n41.then]
    WorldContextObject=@n3
    TextLocation=vector + vector!(A=Get PawnWorldTr(), B_X:double=0, B_Y:double=0, B_Z:double=Get Pawn Capsule Half Height()) -> (ReturnValue)
    Text="Build String (Float)"!(AppendTo:String, Prefix:String="Pawn Speed 2D:    ", InDouble:double=SelectFloat(A=@n43, B=@n42, bPickA=@n40), Suffix:String) -> (ReturnValue:String)
    TextColor=Get PawnColor()
    Duration=0
   )
   > Set DebugTextOffsetZ(DebugTextOffsetZ=float + float(A=Get DebugTextOffsetZ() -> (DebugTextOffsetZ:double), B=15))
   Branch(execute=@n17.then_5, Condition=IsChecked(self=Get DrawPawnTranslation() -> (DrawPawnTranslation:Editor Utility Check Box Object Reference)))
   n44 = GetRightVector!(self:Scene Component Object Reference=Get PawnRootComponent() -> (PawnRootComponent:Scene Component Object Reference)) -> (ReturnValue)
   n45 = vector - vector(A=Get PrevPawnWorldTr() -> (PrevPawnWorldTr:Vector), B=Get PawnWorldTr())
   n46 = "Build String (Float)"(Prefix="Translate Dist:    ", InDouble=Vector Length(A=@n45))
   n47 = Dot Product!(A:Vector=@n45, B:Vector=Get PawnForwardVector()) -> (ReturnValue:double)
   n48 = Get Pawn Capsule Half Height()
   n49 = Get DebugTextOffsetZ()
   > DrawDebugString(
    WorldContextObject=@n3
    TextLocation=vector + vector(A=Get PawnWorldTr(), B_X=0, B_Y=0, B_Z=float + float(A=@n49, B=@n48))
    Text="Build String (Float)"(
     AppendTo="Build String (Float)"(AppendTo=@n46, Prefix="      Forward:    ", InDouble=@n47)
     Prefix="     Side:    "
     InDouble=Dot Product(A=@n45, B=@n44)
    )
    TextColor=Get PawnColor()
    Duration=0
   )
   > Set DebugTextOffsetZ(DebugTextOffsetZ=float + float(A=Get DebugTextOffsetZ(), B=15))
   > Set PrevPawnWorldTr(PrevPawnWorldTr:Vector=Get PawnWorldTr()) -> (Output_Get:Vector)
   Branch(execute=@n17.then_6, Condition=Get IsJunoPlayer() -> (IsJunoPlayer:bool))
   n50 = Get CurrentPawn()
   > n51 = Bad cast node!(execute, Object=GetAnimInstance(self=Get Mesh(self=@n50))) -> (then, CastFailed, AsABP Minifig Main:Object Reference)
   > n52 = Sequence!(execute) -> (then_0, then_1, then_2)
   n53 = Sequence!(execute=@n52.then_2) -> (then_0, then_1, then_2, then_3:Exec, then_4:Exec, then_5:Exec)
   n54 = IsChecked(self=Get DisableProp1() -> (DisableProp1:Editor Utility Check Box Object Reference))
   > Branch(Condition="Not Equal (Boolean)"(A=Get bLayering_V3_Disable_Prop1!(self:Object Reference=@n51) -> (bLayering_V3_Disable_Prop1:bool), B=@n54) -> (ReturnValue=~))
   > Set bLayering_V3_Disable_Prop1!(execute, bLayering_V3_Disable_Prop1:bool=@n54, self:Object Reference=@n51) -> (then, Output_Get:bool)
   n55 = IsChecked(self=Get DisableProp2() -> (DisableProp2:Editor Utility Check Box Object Reference))
   Branch(
    execute=@n53.then_1
    Condition="Not Equal (Boolean)"(A=Get bLayering_V3_Disable_Prop2!(self:Object Reference=@n51) -> (bLayering_V3_Disable_Prop2:bool), B=@n55) -> (ReturnValue=~)
   )
   > Set bLayering_V3_Disable_Prop2!(execute, bLayering_V3_Disable_Prop2:bool=@n55, self:Object Reference=@n51) -> (then, Output_Get:bool)
   n56 = IsChecked(self=Get DisableAddLoco() -> (DisableAddLoco:Editor Utility Check Box Object Reference))
   Branch(
    execute=@n53.then_2
    Condition="Not Equal (Boolean)"(A=Get bLayering_V3_Disable_LocoAdd!(self:Object Reference=@n51) -> (bLayering_V3_Disable_LocoAdd:bool), B=@n56) -> (ReturnValue=~)
   )
   > Set bLayering_V3_Disable_LocoAdd!(execute, bLayering_V3_Disable_LocoAdd:bool=@n56, self:Object Reference=@n51) -> (then, Output_Get:bool)
   n57 = IsChecked(self=Get DisableMeshSpace() -> (DisableMeshSpace:Editor Utility Check Box Object Reference))
   Branch(
    execute=@n53.then_3
    Condition="Not Equal (Boolean)"(A=Get bLayering_V3_Disable_MeshSpace!(self:Object Reference=@n51) -> (bLayering_V3_Disable_MeshSpace:bool), B=@n57) -> (ReturnValue=~)
   )
   > Set bLayering_V3_Disable_MeshSpace!(execute, bLayering_V3_Disable_MeshSpace:bool=@n57, self:Object Reference=@n51) -> (then, Output_Get:bool)
   n58 = IsChecked(self=Get DisableSecondary() -> (DisableSecondary:Editor Utility Check Box Object Reference))
   Branch(
    execute=@n53.then_4
    Condition="Not Equal (Boolean)"(A=Get bLayering_V3_Disable_Secondary!(self:Object Reference=@n51) -> (bLayering_V3_Disable_Secondary:bool), B=@n58) -> (ReturnValue=~)
   )
   > Set bLayering_V3_Disable_Secondary!(execute, bLayering_V3_Disable_Secondary:bool=@n58, self:Object Reference=@n51) -> (then, Output_Get:bool)
   n59 = IsChecked(self=Get DisableAO() -> (DisableAO:Editor Utility Check Box Object Reference))
   Branch(
    execute=@n53.then_5
    Condition="Not Equal (Boolean)"(A=Get bLayering_V3_Disable_AO!(self:Object Reference=@n51) -> (bLayering_V3_Disable_AO:bool), B=@n59) -> (ReturnValue=~)
   )
   > Set bLayering_V3_Disable_AO!(execute, bLayering_V3_Disable_AO:bool=@n59, self:Object Reference=@n51) -> (then, Output_Get:bool)
   n60 = Sequence!(execute=@n17.then_7) -> (then_0, then_1)
   n61 = Get Is Playing Montage() -> (Is Playing Montage:bool)
   n62 = Get PlayMontage() -> (PlayMontage:Editor Utility Check Box Object Reference)
   > Branch(Condition=AND Boolean(A=IsChecked(self=@n62), B=NOT Boolean(A=@n61)))
   n63 = Get CurrentPawn()
   > PlayAnimMontage!(
    execute
    self:Character Object Reference=@n63
    AnimMontage:Anim Montage Object Reference=Get Montage() -> (Montage:Anim Montage Object Reference)
    InPlayRate:float=1
    StartSectionName:Name=None
   ) -> (then, ReturnValue:float)
   > Set Is Playing Montage(Is Playing Montage:bool=true) -> (Output_Get:bool)
   Branch(execute=@n60.then_1, Condition=Get Is Playing Montage())
   > n64 = Is Valid(InputObject=GetCurrentMontage!(self:Character Object Reference=@n63) -> (ReturnValue:Anim Montage Object Reference))
   SetCheckedState!(execute=@n64.Is Not Valid, self:Check Box Object Reference=Get PlayMontage(), InCheckedState:ECheckBoxState Enum=Unchecked) -> (then)
   > Set Is Playing Montage(Is Playing Montage=false)
   n65 = IsChecked()
   Branch(Condition="Not Equal (Boolean)"(A=Get bLayering_V3_Enable!(self:Object Reference=@n51) -> (bLayering_V3_Enable:bool), B=@n65) -> (ReturnValue=~))
   > n66 = Branch(Condition=@n65)
   > ExecuteConsoleCommandFromEditor!(execute, ConsoleCommand:String=Juno.Anim.UseNewLayeringNode 1) -> (then)
   n67 = Get LayeringSettings() -> (LayeringSettings:Grid Panel Object Reference)
   > SetIsEnabled(self=@n67, bInIsEnabled=true)
   > SetIsChecked!(execute, self:Check Box Object Reference, InIsChecked:bool=true) -> (then)
   ExecuteConsoleCommandFromEditor(execute=@n66.else, ConsoleCommand=Juno.Anim.UseNewLayeringNode 0)
   > SetIsEnabled(self=@n67, bInIsEnabled=false)
   > SetIsChecked(InIsChecked=false)
   Event Construct()
   Event Destruct()
  ]}
 Pawn Tick {nodes [Pawn Tick() {is=title K2Node_FunctionEntry}]}
 Reset Pawn Selection {nodes [Reset Pawn Selection() {is=title K2Node_FunctionEntry}]}
]
variables [
 Arrow Length {type=number, default=400}
 CurrentPawn {type=FortPawn, default=null}
 DebugTextOffsetZ {type=number}
 default_combo_box_str {type=string, default="None                   "}
 Draw Pawn Facing {type=boolean}
 Is Playing Montage {type=boolean}
 IsJunoPlayer {type=boolean}
 Look For Pawns of Type {type="array<Class>", default [null,null,null,null,null,null,null]}
 Montage {type=AnimMontage, default=null}
 Pawn Capsule Half Height {type=number}
 Pawn Capsule Radius {type=number}
 Pawn Tick {default=()}
 PawnCapsuleComponent {type=CapsuleComponent, default=null}
 PawnColor {type=LinearColor, default=<rgba 1 0.9757 0.127 1>}
 PawnForwardVector {type=Vector}
 PawnRootComponent {type=SceneComponent, default=null}
 PawnsFound {type="map<FortPawn>", default {}}
 PawnWorldTr {type=Vector}
 PrevPawnWorldTr {type=Vector}
 Reset Pawn Selection {default=()}
 SkelColor {type=LinearColor, default=<rgba 0 0.921 1 1>}
 World {type=Object, default=null}
 World Is Valid {type=boolean}
]
widgets {
 info {rootWidgetClass=&EditorUtilityScrollBox, widgetCount=65}
 widgets [
  EditorUtilityScrollBox_1 {widgetClassPath=&EditorUtilityScrollBox, widget=EditorUtilityScrollBox_1, properties {orientation=Orient_Horizontal}}
  EditorUtilityScrollBox_0 {
   widgetClassPath=&EditorUtilityScrollBox
   parent=EditorUtilityScrollBox_1
   widget=EditorUtilityScrollBox_0
   slot {class=ScrollBoxSlot, properties {size {sizeRule=Fill, value=1}}}
  }
  VerticalBox_1 {
   widgetClassPath=&VerticalBox
   parent=EditorUtilityScrollBox_0
   widget=VerticalBox_1
   slot {class=ScrollBoxSlot, properties {horizontalAlignment=HAlign_Left, verticalAlignment=VAlign_Center, padding=<ltrb 16 4 0 0>}}
  }
  TextBlock_1 {
   widgetClassPath=&TextBlock
   parent=VerticalBox_1
   widget=TextBlock_1
   properties {
    text=Select Pawn:
    colorAndOpacity {colorUseRule=UseColor_Specified, specifiedColor=<rgba 0.1128 0.1128 0.1128 1>}
    font {
     bForceMonospaced=false
     bMaterialIsStencil=false
     fontMaterial=null
     fontObject=&NotoSans
     letterSpacing=0
     monospacedWidth=1
     size=9
     skewAmount=0
     typefaceFontName=NotoSans-Bold
     outlineSettings {
      bApplyOutlineToDropShadows=false
      bMiteredCorners=false
      bSeparateFillAlpha=false
      outlineMaterial=null
      outlineSize=0
      outlineColor=<rgba 0 0 0 1>
     }
    }
   }
   slot {class=VerticalBoxSlot}
  }
  SelectPawnComboBox {
   widgetClassPath=/Script/Blutility.EditorUtilityComboBoxString
   bIsVariable=true
   parent=VerticalBox_1
   widget=SelectPawnComboBox
   properties {
    selectedOption="                   "
    font {
     bForceMonospaced=false
     bMaterialIsStencil=false
     fontMaterial=null
     fontObject=&NotoSans
     letterSpacing=0
     monospacedWidth=1
     size=9
     skewAmount=0
     typefaceFontName=NotoSans-Regular
     outlineSettings {
      bApplyOutlineToDropShadows=false
      bMiteredCorners=false
      bSeparateFillAlpha=false
      outlineMaterial=null
      outlineSize=0
      outlineColor=<rgba 0 0 0 1>
     }
    }
    foregroundColor {colorUseRule=UseColor_Specified, specifiedColor=<rgba 0.5271 0.5271 0.5271 1>}
    widgetStyle {
     comboButtonStyle {
      downArrowAlign=VAlign_Center
      buttonStyle {
       clickedSlateSound {resourceObject=null}
       disabled {
        bIsDynamicallyLoaded=false
        drawAs=RoundedBox
        imageType=NoImage
        mirroring=NoMirror
        resourceName=None
        resourceObject=null
        tiling=NoTile
        imageSize=<0 0>
        margin=<ltrb 0 0 0 0>
        outlineSettings {
         bUseBrushTransparency=false
         roundingType=FixedRadius
         width=1
         color {colorUseRule=UseColor_ColorTable, specifiedColor=<rgba 0 0 0 0>}
         cornerRadii=<4 4 4 4>
        }
        tintColor {colorUseRule=UseColor_Specified, specifiedColor=<rgba 0.0395 0.0395 0.0395 1>}
        uVRegion {bIsValid=false, max=<0 0>, min=<0 0>}
       }
       disabledForeground {colorUseRule=UseColor_ColorTable, specifiedColor=<rgba 0 0 0 0>}
       hovered {
        bIsDynamicallyLoaded=false
        drawAs=RoundedBox
        imageType=NoImage
        mirroring=NoMirror
        resourceName=None
        resourceObject=null
        tiling=NoTile
        imageSize=<0 0>
        margin=<ltrb 0 0 0 0>
        outlineSettings {
         bUseBrushTransparency=false
         roundingType=FixedRadius
         width=1
         color {colorUseRule=UseColor_ColorTable, specifiedColor=<rgba 0 0 0 0>}
         cornerRadii=<4 4 4 4>
        }
        tintColor {colorUseRule=UseColor_Specified, specifiedColor=<rgba 0.0048 0.0048 0.0048 1>}
        uVRegion {bIsValid=false, max=<0 0>, min=<0 0>}
       }
       hoveredForeground {colorUseRule=UseColor_ColorTable, specifiedColor=<rgba 0 0 0 0>}
       hoveredSlateSound {resourceObject=null}
       normal {
        bIsDynamicallyLoaded=false
        drawAs=RoundedBox
        imageType=NoImage
        mirroring=NoMirror
        resourceName=None
        resourceObject=null
        tiling=NoTile
        imageSize=<0 0>
        margin=<ltrb 0 0 0 0>
        outlineSettings {
         bUseBrushTransparency=false
         roundingType=FixedRadius
         width=1
         color {colorUseRule=UseColor_ColorTable, specifiedColor=<rgba 0 0 0 0>}
         cornerRadii=<4 4 4 4>
        }
        tintColor {colorUseRule=UseColor_Specified, specifiedColor=<rgba 0.0048 0.0048 0.0048 1>}
        uVRegion {bIsValid=false, max=<0 0>, min=<0 0>}
       }
       normalForeground {colorUseRule=UseColor_ColorTable, specifiedColor=<rgba 0 0 0 0>}
       normalPadding=<ltrb 8 1 8 1>
       pressed {
        bIsDynamicallyLoaded=false
        drawAs=RoundedBox
        imageType=NoImage
        mirroring=NoMirror
        resourceName=None
        resourceObject=null
        tiling=NoTile
        imageSize=<0 0>
        margin=<ltrb 0 0 0 0>
        outlineSettings {
         bUseBrushTransparency=false
         roundingType=FixedRadius
         width=1
         color {colorUseRule=UseColor_ColorTable, specifiedColor=<rgba 0 0 0 0>}
         cornerRadii=<4 4 4 4>
        }
        tintColor {colorUseRule=UseColor_Specified, specifiedColor=<rgba 0.0048 0.0048 0.0048 1>}
        uVRegion {bIsValid=false, max=<0 0>, min=<0 0>}
       }
       pressedForeground {colorUseRule=UseColor_ColorTable, specifiedColor=<rgba 0 0 0 0>}
       pressedPadding=<ltrb 8 2 8 0>
       pressedSlateSound {resourceObject=null}
      }
      contentPadding=<ltrb 0 0 0 0>
      downArrowImage {
       bIsDynamicallyLoaded=false
       drawAs=Image
       imageType=Vector
       mirroring=NoMirror
       resourceName=../../../Engine/Content/Slate/Starship/CoreWidgets/ComboBox/wide-chevron-down.svg
       resourceObject=null
       tiling=NoTile
       imageSize=<16 16>
       margin=<ltrb 0 0 0 0>
       outlineSettings {
        bUseBrushTransparency=false
        roundingType=HalfHeightRadius
        width=0
        color {colorUseRule=UseColor_Specified, specifiedColor=<rgba 0 0 0 0>}
        cornerRadii=<0 0 0 0>
       }
       tintColor {colorUseRule=UseColor_Specified, specifiedColor=<rgba 1 1 1 1>}
       uVRegion {bIsValid=false, max=<0 0>, min=<0 0>}
      }
      downArrowPadding=<ltrb 2 2 2 2>
      menuBorderBrush {
       bIsDynamicallyLoaded=false
       drawAs=RoundedBox
       imageType=NoImage
       mirroring=NoMirror
       resourceName=None
       resourceObject=null
       tiling=NoTile
       imageSize=<0 0>
       margin=<ltrb 0 0 0 0>
       outlineSettings {
        bUseBrushTransparency=false
        roundingType=FixedRadius
        width=1
        color {colorUseRule=UseColor_ColorTable, specifiedColor=<rgba 0 0 0 0>}
        cornerRadii=<0 0 0 0>
       }
       tintColor {colorUseRule=UseColor_Specified, specifiedColor=<rgba 0.0395 0.0395 0.0395 1>}
       uVRegion {bIsValid=false, max=<0 0>, min=<0 0>}
      }
      menuBorderPadding=<ltrb 0 0 0 0>
      shadowColorAndOpacity=<rgba 0 0 0 1>
      shadowOffset=<0 0>
     }
     contentPadding=<ltrb 0 0 0 0>
     menuRowPadding=<ltrb 8 0 8 0>
     pressedSlateSound {resourceObject=null}
     selectionChangeSlateSound {resourceObject=null}
    }
   }
   slot {class=VerticalBoxSlot}
  }
  PawnBlockBorder {
   widgetClassPath=&Border
   bIsVariable=true
   parent=EditorUtilityScrollBox_0
   widget=PawnBlockBorder
   properties {
    bIsEnabled=false
    background {
     bIsDynamicallyLoaded=false
     drawAs=RoundedBox
     imageType=NoImage
     mirroring=NoMirror
     resourceName=None
     resourceObject=null
     tiling=NoTile
     imageSize=<32 32>
     margin=<ltrb 0 0 0 0>
     outlineSettings {
      bUseBrushTransparency=false
      roundingType=FixedRadius
      width=1
      color {colorUseRule=UseColor_Specified, specifiedColor=<rgba 0 1 0 0.644>}
      cornerRadii=<4 4 4 4>
     }
     tintColor {colorUseRule=UseColor_Specified, specifiedColor=<rgba 0.0226 0.0226 0.0226 1>}
     uVRegion {bIsValid=false, max=<0 0>, min=<0 0>}
    }
    padding=<ltrb 8 4 8 8>
   }
   slot {class=ScrollBoxSlot, properties {padding=<ltrb 16 8 8 8>}}
  }
  VerticalBox_0 {
   widgetClassPath=&VerticalBox
   parent=PawnBlockBorder
   widget=VerticalBox_0
   slot {class=BorderSlot, properties {padding=<ltrb 8 4 8 8>}}
  }
  HorizontalBox_0 {
   widgetClassPath=&HorizontalBox
   parent=VerticalBox_0
   widget=HorizontalBox_0
   slot {class=VerticalBoxSlot, properties {size {sizeRule=Fill, value=1}}}
  }
  TextBlock_2 {
   widgetClassPath=&TextBlock
   parent=HorizontalBox_0
   widget=TextBlock_2
   properties {
    text=Pawn Debug :
    colorAndOpacity {colorUseRule=UseColor_Specified, specifiedColor=<rgba 0.1128 0.1128 0.1128 1>}
    font {
     bForceMonospaced=false
     bMaterialIsStencil=false
     fontMaterial=null
     fontObject=&NotoSans
     letterSpacing=0
     monospacedWidth=1
     size=9
     skewAmount=0
     typefaceFontName=NotoSans-Bold
     outlineSettings {
      bApplyOutlineToDropShadows=false
      bMiteredCorners=false
      bSeparateFillAlpha=false
      outlineMaterial=null
      outlineSize=0
      outlineColor=<rgba 0 0 0 1>
     }
    }
   }
   slot {class=HorizontalBoxSlot}
  }
  Spacer_64 {
   widgetClassPath=&Spacer
   parent=HorizontalBox_0
   widget=Spacer_64
   slot {class=HorizontalBoxSlot, properties {size {sizeRule=Fill, value=1}}}
  }
  TextBlock_3 {
   widgetClassPath=&TextBlock
   parent=HorizontalBox_0
   widget=TextBlock_3
   properties {
    text=?
    toolTipText="Pawn is an underlying entity that chracter is parented to. \r\n\r\nOn creatures it's controlled by AI and Physics.\r\n\r\nOn player it's controlled by Gameplay Code."
    colorAndOpacity {colorUseRule=UseColor_Specified, specifiedColor=<rgba 0 1 0 0.644>}
    font {
     bForceMonospaced=false
     bMaterialIsStencil=false
     fontMaterial=null
     fontObject=&NotoSans
     letterSpacing=0
     monospacedWidth=1
     size=9
     skewAmount=0
     typefaceFontName=NotoSans-Bold
     outlineSettings {
      bApplyOutlineToDropShadows=false
      bMiteredCorners=false
      bSeparateFillAlpha=false
      outlineMaterial=null
      outlineSize=0
      outlineColor=<rgba 0 0 0 1>
     }
    }
   }
   slot {class=HorizontalBoxSlot}
  }
  GridPanel_134 {widgetClassPath=&GridPanel, parent=VerticalBox_0, widget=GridPanel_134, slot {class=VerticalBoxSlot}}
  DrawPawnFacing {
   widgetClassPath=&EditorUtilityCheckBox
   bIsVariable=true
   parent=GridPanel_134
   widget=DrawPawnFacing
   properties {horizontalAlignment=HAlign_Center, renderTransform {angle=0, scale=<0.8 0.8>, shear=<0 0>, translation=<0 0>}}
   slot {class=GridSlot, properties {horizontalAlignment=HAlign_Center, verticalAlignment=VAlign_Center, padding=<ltrb 0 0 2 0>}}
  }
  DrawPawnSpeed2D {
   widgetClassPath=&EditorUtilityCheckBox
   bIsVariable=true
   parent=GridPanel_134
   widget=DrawPawnSpeed2D
   properties {horizontalAlignment=HAlign_Center, renderTransform {angle=0, scale=<0.8 0.8>, shear=<0 0>, translation=<0 0>}}
   slot {class=GridSlot, properties {horizontalAlignment=HAlign_Center, row=1, verticalAlignment=VAlign_Center, padding=<ltrb 0 0 2 0>}}
  }
  DrawPawnTranslation {
   widgetClassPath=&EditorUtilityCheckBox
   bIsVariable=true
   parent=GridPanel_134
   widget=DrawPawnTranslation
   properties {horizontalAlignment=HAlign_Center, renderTransform {angle=0, scale=<0.8 0.8>, shear=<0 0>, translation=<0 0>}}
   slot {class=GridSlot, properties {horizontalAlignment=HAlign_Center, row=2, verticalAlignment=VAlign_Center, padding=<ltrb 0 0 2 0>}}
  }
  DrawPawnCapsule {
   widgetClassPath=&EditorUtilityCheckBox
   bIsVariable=true
   parent=GridPanel_134
   widget=DrawPawnCapsule
   properties {horizontalAlignment=HAlign_Center, renderTransform {angle=0, scale=<0.8 0.8>, shear=<0 0>, translation=<0 0>}}
   slot {class=GridSlot, properties {horizontalAlignment=HAlign_Center, row=3, verticalAlignment=VAlign_Center, padding=<ltrb 0 0 2 0>}}
  }
  TextBlock {
   widgetClassPath=&TextBlock
   parent=GridPanel_134
   widget=TextBlock
   properties {
    text=Facing
    font {
     bForceMonospaced=false
     bMaterialIsStencil=false
     fontMaterial=null
     fontObject=&NotoSans
     letterSpacing=0
     monospacedWidth=1
     size=9.75
     skewAmount=0
     typefaceFontName=NotoSans-Regular
     outlineSettings {
      bApplyOutlineToDropShadows=false
      bMiteredCorners=false
      bSeparateFillAlpha=false
      outlineMaterial=null
      outlineSize=0
      outlineColor=<rgba 0 0 0 1>
     }
    }
   }
   slot {class=GridSlot, properties {column=1, horizontalAlignment=HAlign_Left, verticalAlignment=VAlign_Center}}
  }
  TextBlock_5 {
   widgetClassPath=&TextBlock
   parent=GridPanel_134
   widget=TextBlock_5
   properties {
    text=Pawn Speed 2D
    font {
     bForceMonospaced=false
     bMaterialIsStencil=false
     fontMaterial=null
     fontObject=&NotoSans
     letterSpacing=0
     monospacedWidth=1
     size=9.75
     skewAmount=0
     typefaceFontName=NotoSans-Regular
     outlineSettings {
      bApplyOutlineToDropShadows=false
      bMiteredCorners=false
      bSeparateFillAlpha=false
      outlineMaterial=null
      outlineSize=0
      outlineColor=<rgba 0 0 0 1>
     }
    }
   }
   slot {class=GridSlot, properties {column=1, horizontalAlignment=HAlign_Left, row=1, verticalAlignment=VAlign_Center}}
  }
  TextBlock_37 {
   widgetClassPath=&TextBlock
   parent=GridPanel_134
   widget=TextBlock_37
   properties {
    text="Translation (last tick)"
    font {
     bForceMonospaced=false
     bMaterialIsStencil=false
     fontMaterial=null
     fontObject=&NotoSans
     letterSpacing=0
     monospacedWidth=1
     size=9.75
     skewAmount=0
     typefaceFontName=NotoSans-Regular
     outlineSettings {
      bApplyOutlineToDropShadows=false
      bMiteredCorners=false
      bSeparateFillAlpha=false
      outlineMaterial=null
      outlineSize=0
      outlineColor=<rgba 0 0 0 1>
     }
    }
   }
   slot {class=GridSlot, properties {column=1, horizontalAlignment=HAlign_Left, row=2, verticalAlignment=VAlign_Center}}
  }
  TextBlock_4 {
   widgetClassPath=&TextBlock
   parent=GridPanel_134
   widget=TextBlock_4
   properties {
    text=Collision Capsule
    font {
     bForceMonospaced=false
     bMaterialIsStencil=false
     fontMaterial=null
     fontObject=&NotoSans
     letterSpacing=0
     monospacedWidth=1
     size=9.75
     skewAmount=0
     typefaceFontName=NotoSans-Regular
     outlineSettings {
      bApplyOutlineToDropShadows=false
      bMiteredCorners=false
      bSeparateFillAlpha=false
      outlineMaterial=null
      outlineSize=0
      outlineColor=<rgba 0 0 0 1>
     }
    }
   }
   slot {class=GridSlot, properties {column=1, horizontalAlignment=HAlign_Left, row=3, verticalAlignment=VAlign_Center}}
  }
  SkelBlockBorder {
   widgetClassPath=&Border
   bIsVariable=true
   parent=EditorUtilityScrollBox_0
   widget=SkelBlockBorder
   properties {
    bIsEnabled=false
    background {
     bIsDynamicallyLoaded=false
     drawAs=RoundedBox
     imageType=NoImage
     mirroring=NoMirror
     resourceName=None
     resourceObject=null
     tiling=NoTile
     imageSize=<32 32>
     margin=<ltrb 0 0 0 0>
     outlineSettings {
      bUseBrushTransparency=false
      roundingType=FixedRadius
      width=1
      color {colorUseRule=UseColor_Specified, specifiedColor=<rgba 0 1 0 0.644>}
      cornerRadii=<4 4 4 4>
     }
     tintColor {colorUseRule=UseColor_Specified, specifiedColor=<rgba 0.0226 0.0226 0.0226 1>}
     uVRegion {bIsValid=false, max=<0 0>, min=<0 0>}
    }
    padding=<ltrb 8 4 8 8>
   }
   slot {class=ScrollBoxSlot, properties {padding=<ltrb 16 8 8 8>}}
  }
  VerticalBox {
   widgetClassPath=&VerticalBox
   parent=SkelBlockBorder
   widget=VerticalBox
   slot {class=BorderSlot, properties {padding=<ltrb 8 4 8 8>}}
  }
  HorizontalBox {
   widgetClassPath=&HorizontalBox
   parent=VerticalBox
   widget=HorizontalBox
   slot {class=VerticalBoxSlot, properties {size {sizeRule=Fill, value=1}}}
  }
  TextBlock_6 {
   widgetClassPath=&TextBlock
   parent=HorizontalBox
   widget=TextBlock_6
   properties {
    text=Skel Debug :
    colorAndOpacity {colorUseRule=UseColor_Specified, specifiedColor=<rgba 0.1128 0.1128 0.1128 1>}
    font {
     bForceMonospaced=false
     bMaterialIsStencil=false
     fontMaterial=null
     fontObject=&NotoSans
     letterSpacing=0
     monospacedWidth=1
     size=9
     skewAmount=0
     typefaceFontName=NotoSans-Bold
     outlineSettings {
      bApplyOutlineToDropShadows=false
      bMiteredCorners=false
      bSeparateFillAlpha=false
      outlineMaterial=null
      outlineSize=0
      outlineColor=<rgba 0 0 0 1>
     }
    }
   }
   slot {class=HorizontalBoxSlot}
  }
  Spacer {
   widgetClassPath=&Spacer
   parent=HorizontalBox
   widget=Spacer
   slot {class=HorizontalBoxSlot, properties {size {sizeRule=Fill, value=1}}}
  }
  TextBlock_7 {
   widgetClassPath=&TextBlock
   parent=HorizontalBox
   widget=TextBlock_7
   properties {
    text=?
    toolTipText="Skeleton is a child of a Pawn. \r\n\r\nSkeleton has it's own 'root bone that can be used to add small offsets on top of Pawn movement to improve quailty of animation.\r\n\r\nSkelton is controlled by Animation Blueprint. "
    colorAndOpacity {colorUseRule=UseColor_Specified, specifiedColor=<rgba 0 1 0 0.644>}
    font {
     bForceMonospaced=false
     bMaterialIsStencil=false
     fontMaterial=null
     fontObject=&NotoSans
     letterSpacing=0
     monospacedWidth=1
     size=9
     skewAmount=0
     typefaceFontName=NotoSans-Bold
     outlineSettings {
      bApplyOutlineToDropShadows=false
      bMiteredCorners=false
      bSeparateFillAlpha=false
      outlineMaterial=null
      outlineSize=0
      outlineColor=<rgba 0 0 0 1>
     }
    }
   }
   slot {class=HorizontalBoxSlot}
  }
  GridPanel {widgetClassPath=&GridPanel, parent=VerticalBox, widget=GridPanel, slot {class=VerticalBoxSlot}}
  DrawSkelFacing {
   widgetClassPath=&EditorUtilityCheckBox
   bIsVariable=true
   parent=GridPanel
   widget=DrawSkelFacing
   properties {horizontalAlignment=HAlign_Center, renderTransform {angle=0, scale=<0.8 0.8>, shear=<0 0>, translation=<0 0>}}
   slot {class=GridSlot, properties {horizontalAlignment=HAlign_Center, verticalAlignment=VAlign_Center, padding=<ltrb 0 0 2 0>}}
  }
  DrawSkelSpeed2D {
   widgetClassPath=&EditorUtilityCheckBox
   bIsVariable=true
   parent=GridPanel
   widget=DrawSkelSpeed2D
   properties {horizontalAlignment=HAlign_Center, renderTransform {angle=0, scale=<0.8 0.8>, shear=<0 0>, translation=<0 0>}}
   slot {class=GridSlot, properties {horizontalAlignment=HAlign_Center, row=1, verticalAlignment=VAlign_Center, padding=<ltrb 0 0 2 0>}}
  }
  TextBlock_8 {
   widgetClassPath=&TextBlock
   parent=GridPanel
   widget=TextBlock_8
   properties {
    text=Facing Offset
    font {
     bForceMonospaced=false
     bMaterialIsStencil=false
     fontMaterial=null
     fontObject=&NotoSans
     letterSpacing=0
     monospacedWidth=1
     size=9.75
     skewAmount=0
     typefaceFontName=NotoSans-Regular
     outlineSettings {
      bApplyOutlineToDropShadows=false
      bMiteredCorners=false
      bSeparateFillAlpha=false
      outlineMaterial=null
      outlineSize=0
      outlineColor=<rgba 0 0 0 1>
     }
    }
   }
   slot {class=GridSlot, properties {column=1, horizontalAlignment=HAlign_Left, verticalAlignment=VAlign_Center}}
  }
  TextBlock_9 {
   widgetClassPath=&TextBlock
   parent=GridPanel
   widget=TextBlock_9
   properties {
    text=Translation Offset
    font {
     bForceMonospaced=false
     bMaterialIsStencil=false
     fontMaterial=null
     fontObject=&NotoSans
     letterSpacing=0
     monospacedWidth=1
     size=9.75
     skewAmount=0
     typefaceFontName=NotoSans-Regular
     outlineSettings {
      bApplyOutlineToDropShadows=false
      bMiteredCorners=false
      bSeparateFillAlpha=false
      outlineMaterial=null
      outlineSize=0
      outlineColor=<rgba 0 0 0 1>
     }
    }
   }
   slot {class=GridSlot, properties {column=1, horizontalAlignment=HAlign_Left, row=1, verticalAlignment=VAlign_Center}}
  }
  JunoPlayerBlockBorder {
   widgetClassPath=&Border
   bIsVariable=true
   parent=EditorUtilityScrollBox_0
   widget=JunoPlayerBlockBorder
   properties {
    visibility=Collapsed
    background {
     bIsDynamicallyLoaded=false
     drawAs=RoundedBox
     imageType=NoImage
     mirroring=NoMirror
     resourceName=None
     resourceObject=null
     tiling=NoTile
     imageSize=<32 32>
     margin=<ltrb 0 0 0 0>
     outlineSettings {
      bUseBrushTransparency=false
      roundingType=FixedRadius
      width=1
      color {colorUseRule=UseColor_Specified, specifiedColor=<rgba 0 1 0 0.644>}
      cornerRadii=<4 4 4 4>
     }
     tintColor {colorUseRule=UseColor_Specified, specifiedColor=<rgba 0.0226 0.0226 0.0226 1>}
     uVRegion {bIsValid=false, max=<0 0>, min=<0 0>}
    }
    padding=<ltrb 8 4 8 8>
   }
   slot {class=ScrollBoxSlot, properties {padding=<ltrb 16 8 8 8>}}
  }
  VerticalBox_2 {
   widgetClassPath=&VerticalBox
   parent=JunoPlayerBlockBorder
   widget=VerticalBox_2
   slot {class=BorderSlot, properties {padding=<ltrb 8 4 8 8>}}
  }
  HorizontalBox_1 {
   widgetClassPath=&HorizontalBox
   parent=VerticalBox_2
   widget=HorizontalBox_1
   slot {class=VerticalBoxSlot, properties {size {sizeRule=Fill, value=1}}}
  }
  TextBlock_11 {
   widgetClassPath=&TextBlock
   parent=HorizontalBox_1
   widget=TextBlock_11
   properties {
    text=Juno Player:
    colorAndOpacity {colorUseRule=UseColor_Specified, specifiedColor=<rgba 0.1128 0.1128 0.1128 1>}
    font {
     bForceMonospaced=false
     bMaterialIsStencil=false
     fontMaterial=null
     fontObject=&NotoSans
     letterSpacing=0
     monospacedWidth=1
     size=9
     skewAmount=0
     typefaceFontName=NotoSans-Bold
     outlineSettings {
      bApplyOutlineToDropShadows=false
      bMiteredCorners=false
      bSeparateFillAlpha=false
      outlineMaterial=null
      outlineSize=0
      outlineColor=<rgba 0 0 0 1>
     }
    }
   }
   slot {class=HorizontalBoxSlot}
  }
  Spacer_1 {
   widgetClassPath=&Spacer
   parent=HorizontalBox_1
   widget=Spacer_1
   slot {class=HorizontalBoxSlot, properties {size {sizeRule=Fill, value=1}}}
  }
  TextBlock_12 {
   widgetClassPath=&TextBlock
   parent=HorizontalBox_1
   widget=TextBlock_12
   properties {
    text=?
    toolTipText=Debug specific to a Juno Player
    colorAndOpacity {colorUseRule=UseColor_Specified, specifiedColor=<rgba 0 1 0 0.644>}
    font {
     bForceMonospaced=false
     bMaterialIsStencil=false
     fontMaterial=null
     fontObject=&NotoSans
     letterSpacing=0
     monospacedWidth=1
     size=9
     skewAmount=0
     typefaceFontName=NotoSans-Bold
     outlineSettings {
      bApplyOutlineToDropShadows=false
      bMiteredCorners=false
      bSeparateFillAlpha=false
      outlineMaterial=null
      outlineSize=0
      outlineColor=<rgba 0 0 0 1>
     }
    }
   }
   slot {class=HorizontalBoxSlot}
  }
  GridPanel_1 {widgetClassPath=&GridPanel, parent=VerticalBox_2, widget=GridPanel_1, slot {class=VerticalBoxSlot}}
  LayeringTweakables {
   widgetClassPath=&TextBlock
   parent=GridPanel_1
   widget=LayeringTweakables
   properties {
    text=Layering Tweaks:
    font {
     bForceMonospaced=false
     bMaterialIsStencil=false
     fontMaterial=null
     fontObject=&NotoSans
     letterSpacing=0
     monospacedWidth=1
     size=9.75
     skewAmount=0
     typefaceFontName=NotoSans-Regular
     outlineSettings {
      bApplyOutlineToDropShadows=false
      bMiteredCorners=false
      bSeparateFillAlpha=false
      outlineMaterial=null
      outlineSize=0
      outlineColor=<rgba 0 0 0 1>
     }
    }
   }
   slot {class=GridSlot, properties {column=1, horizontalAlignment=HAlign_Left, verticalAlignment=VAlign_Center}}
  }
  Spacer_101 {
   widgetClassPath=&Spacer
   parent=GridPanel_1
   widget=Spacer_101
   properties {size=<1 6>}
   slot {class=GridSlot, properties {row=1}}
  }
  LayeringSettings {
   widgetClassPath=&GridPanel
   bIsVariable=true
   parent=GridPanel_1
   widget=LayeringSettings
   slot {class=GridSlot, properties {column=1, row=2, padding=<ltrb 10 0 0 0>}}
  }
  DisableProp1 {
   widgetClassPath=&EditorUtilityCheckBox
   bIsVariable=true
   parent=LayeringSettings
   widget=DisableProp1
   properties {horizontalAlignment=HAlign_Center, renderTransform {angle=0, scale=<0.8 0.8>, shear=<0 0>, translation=<0 0>}}
   slot {class=GridSlot, properties {horizontalAlignment=HAlign_Center, verticalAlignment=VAlign_Center, padding=<ltrb 0 0 2 0>}}
  }
  TextBlock_10 {
   widgetClassPath=&TextBlock
   parent=LayeringSettings
   widget=TextBlock_10
   properties {
    text=Disable Prop1
    font {
     bForceMonospaced=false
     bMaterialIsStencil=false
     fontMaterial=null
     fontObject=&NotoSans
     letterSpacing=0
     monospacedWidth=1
     size=9.75
     skewAmount=0
     typefaceFontName=NotoSans-Regular
     outlineSettings {
      bApplyOutlineToDropShadows=false
      bMiteredCorners=false
      bSeparateFillAlpha=false
      outlineMaterial=null
      outlineSize=0
      outlineColor=<rgba 0 0 0 1>
     }
    }
   }
   slot {class=GridSlot, properties {column=1, horizontalAlignment=HAlign_Left, verticalAlignment=VAlign_Center}}
  }
  DisableProp2 {
   widgetClassPath=&EditorUtilityCheckBox
   bIsVariable=true
   parent=LayeringSettings
   widget=DisableProp2
   properties {horizontalAlignment=HAlign_Center, renderTransform {angle=0, scale=<0.8 0.8>, shear=<0 0>, translation=<0 0>}}
   slot {class=GridSlot, properties {horizontalAlignment=HAlign_Center, row=1, verticalAlignment=VAlign_Center, padding=<ltrb 0 0 2 0>}}
  }
  TextBlock_14 {
   widgetClassPath=&TextBlock
   parent=LayeringSettings
   widget=TextBlock_14
   properties {
    text=Disable Prop2
    font {
     bForceMonospaced=false
     bMaterialIsStencil=false
     fontMaterial=null
     fontObject=&NotoSans
     letterSpacing=0
     monospacedWidth=1
     size=9.75
     skewAmount=0
     typefaceFontName=NotoSans-Regular
     outlineSettings {
      bApplyOutlineToDropShadows=false
      bMiteredCorners=false
      bSeparateFillAlpha=false
      outlineMaterial=null
      outlineSize=0
      outlineColor=<rgba 0 0 0 1>
     }
    }
   }
   slot {class=GridSlot, properties {column=1, horizontalAlignment=HAlign_Left, row=1, verticalAlignment=VAlign_Center}}
  }
  DisableAddLoco {
   widgetClassPath=&EditorUtilityCheckBox
   bIsVariable=true
   parent=LayeringSettings
   widget=DisableAddLoco
   properties {horizontalAlignment=HAlign_Center, renderTransform {angle=0, scale=<0.8 0.8>, shear=<0 0>, translation=<0 0>}}
   slot {class=GridSlot, properties {horizontalAlignment=HAlign_Center, row=2, verticalAlignment=VAlign_Center, padding=<ltrb 0 0 2 0>}}
  }
  TextBlock_15 {
   widgetClassPath=&TextBlock
   parent=LayeringSettings
   widget=TextBlock_15
   properties {
    text=Disable Add Loco
    font {
     bForceMonospaced=false
     bMaterialIsStencil=false
     fontMaterial=null
     fontObject=&NotoSans
     letterSpacing=0
     monospacedWidth=1
     size=9.75
     skewAmount=0
     typefaceFontName=NotoSans-Regular
     outlineSettings {
      bApplyOutlineToDropShadows=false
      bMiteredCorners=false
      bSeparateFillAlpha=false
      outlineMaterial=null
      outlineSize=0
      outlineColor=<rgba 0 0 0 1>
     }
    }
   }
   slot {class=GridSlot, properties {column=1, horizontalAlignment=HAlign_Left, row=2, verticalAlignment=VAlign_Center}}
  }
  DisableMeshSpace {
   widgetClassPath=&EditorUtilityCheckBox
   bIsVariable=true
   parent=LayeringSettings
   widget=DisableMeshSpace
   properties {horizontalAlignment=HAlign_Center, renderTransform {angle=0, scale=<0.8 0.8>, shear=<0 0>, translation=<0 0>}}
   slot {class=GridSlot, properties {horizontalAlignment=HAlign_Center, row=3, verticalAlignment=VAlign_Center, padding=<ltrb 0 0 2 0>}}
  }
  TextBlock_16 {
   widgetClassPath=&TextBlock
   parent=LayeringSettings
   widget=TextBlock_16
   properties {
    text=Disable MeshSpace
    font {
     bForceMonospaced=false
     bMaterialIsStencil=false
     fontMaterial=null
     fontObject=&NotoSans
     letterSpacing=0
     monospacedWidth=1
     size=9.75
     skewAmount=0
     typefaceFontName=NotoSans-Regular
     outlineSettings {
      bApplyOutlineToDropShadows=false
      bMiteredCorners=false
      bSeparateFillAlpha=false
      outlineMaterial=null
      outlineSize=0
      outlineColor=<rgba 0 0 0 1>
     }
    }
   }
   slot {class=GridSlot, properties {column=1, horizontalAlignment=HAlign_Left, row=3, verticalAlignment=VAlign_Center}}
  }
  DisableSecondary {
   widgetClassPath=&EditorUtilityCheckBox
   bIsVariable=true
   parent=LayeringSettings
   widget=DisableSecondary
   properties {horizontalAlignment=HAlign_Center, renderTransform {angle=0, scale=<0.8 0.8>, shear=<0 0>, translation=<0 0>}}
   slot {class=GridSlot, properties {horizontalAlignment=HAlign_Center, row=4, verticalAlignment=VAlign_Center, padding=<ltrb 0 0 2 0>}}
  }
  TextBlock_18 {
   widgetClassPath=&TextBlock
   parent=LayeringSettings
   widget=TextBlock_18
   properties {
    text=Disable Secondary
    font {
     bForceMonospaced=false
     bMaterialIsStencil=false
     fontMaterial=null
     fontObject=&NotoSans
     letterSpacing=0
     monospacedWidth=1
     size=9.75
     skewAmount=0
     typefaceFontName=NotoSans-Regular
     outlineSettings {
      bApplyOutlineToDropShadows=false
      bMiteredCorners=false
      bSeparateFillAlpha=false
      outlineMaterial=null
      outlineSize=0
      outlineColor=<rgba 0 0 0 1>
     }
    }
   }
   slot {class=GridSlot, properties {column=1, horizontalAlignment=HAlign_Left, row=4, verticalAlignment=VAlign_Center}}
  }
  DisableAO {
   widgetClassPath=&EditorUtilityCheckBox
   bIsVariable=true
   parent=LayeringSettings
   widget=DisableAO
   properties {horizontalAlignment=HAlign_Center, renderTransform {angle=0, scale=<0.8 0.8>, shear=<0 0>, translation=<0 0>}}
   slot {class=GridSlot, properties {horizontalAlignment=HAlign_Center, row=5, verticalAlignment=VAlign_Center, padding=<ltrb 0 0 2 0>}}
  }
  TextBlock_17 {
   widgetClassPath=&TextBlock
   parent=LayeringSettings
   widget=TextBlock_17
   properties {
    text=Disable AO
    font {
     bForceMonospaced=false
     bMaterialIsStencil=false
     fontMaterial=null
     fontObject=&NotoSans
     letterSpacing=0
     monospacedWidth=1
     size=9.75
     skewAmount=0
     typefaceFontName=NotoSans-Regular
     outlineSettings {
      bApplyOutlineToDropShadows=false
      bMiteredCorners=false
      bSeparateFillAlpha=false
      outlineMaterial=null
      outlineSize=0
      outlineColor=<rgba 0 0 0 1>
     }
    }
   }
   slot {class=GridSlot, properties {column=1, row=5}}
  }
  TextBlock_19 {
   widgetClassPath=&TextBlock
   parent=LayeringSettings
   widget=TextBlock_19
   properties {
    text=Test
    visibility=Hidden
    font {
     bForceMonospaced=false
     bMaterialIsStencil=false
     fontMaterial=null
     fontObject=&NotoSans
     letterSpacing=0
     monospacedWidth=1
     size=9.75
     skewAmount=0
     typefaceFontName=NotoSans-Regular
     outlineSettings {
      bApplyOutlineToDropShadows=false
      bMiteredCorners=false
      bSeparateFillAlpha=false
      outlineMaterial=null
      outlineSize=0
      outlineColor=<rgba 0 0 0 1>
     }
    }
   }
   slot {class=GridSlot, properties {column=1, row=6}}
  }
  DebugMontageBlockBorder {
   widgetClassPath=&Border
   bIsVariable=true
   parent=EditorUtilityScrollBox_0
   widget=DebugMontageBlockBorder
   properties {
    bIsEnabled=false
    background {
     bIsDynamicallyLoaded=false
     drawAs=RoundedBox
     imageType=NoImage
     mirroring=NoMirror
     resourceName=None
     resourceObject=null
     tiling=NoTile
     imageSize=<32 32>
     margin=<ltrb 0 0 0 0>
     outlineSettings {
      bUseBrushTransparency=false
      roundingType=FixedRadius
      width=1
      color {colorUseRule=UseColor_Specified, specifiedColor=<rgba 0 1 0 0.644>}
      cornerRadii=<4 4 4 4>
     }
     tintColor {colorUseRule=UseColor_Specified, specifiedColor=<rgba 0.0226 0.0226 0.0226 1>}
     uVRegion {bIsValid=false, max=<0 0>, min=<0 0>}
    }
    padding=<ltrb 8 4 8 8>
   }
   slot {class=ScrollBoxSlot, properties {padding=<ltrb 16 8 8 8>}}
  }
  VerticalBox_4 {
   widgetClassPath=&VerticalBox
   parent=DebugMontageBlockBorder
   widget=VerticalBox_4
   slot {class=BorderSlot, properties {padding=<ltrb 8 4 8 8>}}
  }
  HorizontalBox_2 {
   widgetClassPath=&HorizontalBox
   parent=VerticalBox_4
   widget=HorizontalBox_2
   slot {class=VerticalBoxSlot, properties {size {sizeRule=Fill, value=1}}}
  }
  TextBlock_20 {
   widgetClassPath=&TextBlock
   parent=HorizontalBox_2
   widget=TextBlock_20
   properties {
    text=Debug Montage
    colorAndOpacity {colorUseRule=UseColor_Specified, specifiedColor=<rgba 0.1128 0.1128 0.1128 1>}
    font {
     bForceMonospaced=false
     bMaterialIsStencil=false
     fontMaterial=null
     fontObject=&NotoSans
     letterSpacing=0
     monospacedWidth=1
     size=9
     skewAmount=0
     typefaceFontName=NotoSans-Bold
     outlineSettings {
      bApplyOutlineToDropShadows=false
      bMiteredCorners=false
      bSeparateFillAlpha=false
      outlineMaterial=null
      outlineSize=0
      outlineColor=<rgba 0 0 0 1>
     }
    }
   }
   slot {class=HorizontalBoxSlot}
  }
  Spacer_2 {
   widgetClassPath=&Spacer
   parent=HorizontalBox_2
   widget=Spacer_2
   slot {class=HorizontalBoxSlot, properties {size {sizeRule=Fill, value=1}}}
  }
  TextBlock_21 {
   widgetClassPath=&TextBlock
   parent=HorizontalBox_2
   widget=TextBlock_21
   properties {
    text=?
    toolTipText="Skeleton is a child of a Pawn. \r\n\r\nSkeleton has it's own 'root bone that can be used to add small offsets on top of Pawn movement to improve quailty of animation.\r\n\r\nSkelton is controlled by Animation Blueprint. "
    colorAndOpacity {colorUseRule=UseColor_Specified, specifiedColor=<rgba 0 1 0 0.644>}
    font {
     bForceMonospaced=false
     bMaterialIsStencil=false
     fontMaterial=null
     fontObject=&NotoSans
     letterSpacing=0
     monospacedWidth=1
     size=9
     skewAmount=0
     typefaceFontName=NotoSans-Bold
     outlineSettings {
      bApplyOutlineToDropShadows=false
      bMiteredCorners=false
      bSeparateFillAlpha=false
      outlineMaterial=null
      outlineSize=0
      outlineColor=<rgba 0 0 0 1>
     }
    }
   }
   slot {class=HorizontalBoxSlot}
  }
  VerticalBox_3 {
   widgetClassPath=&VerticalBox
   parent=VerticalBox_4
   widget=VerticalBox_3
   slot {class=VerticalBoxSlot, properties {padding=<ltrb 0 10 0 0>}}
  }
  MontagePropView {
   widgetClassPath=&SinglePropertyView
   parent=VerticalBox_3
   widget=MontagePropView
   properties {propertyName=Montage, toolTipText=Test Montage to push on player}
   slot {class=VerticalBoxSlot}
  }
  HorizontalBox_55 {
   widgetClassPath=&HorizontalBox
   parent=VerticalBox_3
   widget=HorizontalBox_55
   slot {class=VerticalBoxSlot, properties {horizontalAlignment=HAlign_Center}}
  }
  PlayMontage {
   widgetClassPath=&EditorUtilityCheckBox
   bIsVariable=true
   parent=HorizontalBox_55
   widget=PlayMontage
   properties {horizontalAlignment=HAlign_Center, renderTransform {angle=0, scale=<0.8 0.8>, shear=<0 0>, translation=<0 0>}}
   slot {class=HorizontalBoxSlot, properties {horizontalAlignment=HAlign_Center, verticalAlignment=VAlign_Center, padding=<ltrb 0 0 2 0>}}
  }
  TextBlock_13 {
   widgetClassPath=&TextBlock
   parent=HorizontalBox_55
   widget=TextBlock_13
   properties {
    text=Play
    font {
     bForceMonospaced=false
     bMaterialIsStencil=false
     fontMaterial=null
     fontObject=&NotoSans
     letterSpacing=0
     monospacedWidth=1
     size=9.75
     skewAmount=0
     typefaceFontName=NotoSans-Regular
     outlineSettings {
      bApplyOutlineToDropShadows=false
      bMiteredCorners=false
      bSeparateFillAlpha=false
      outlineMaterial=null
      outlineSize=0
      outlineColor=<rgba 0 0 0 1>
     }
    }
   }
   slot {class=HorizontalBoxSlot}
  }
 ]
}
